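# Sound cue for a medium constructed door sound in Fort building, played under the player construction sound class.
# A single wave player feeds a modulator at the root, which is left at its default settings.
# It uses the shared Fort doors attenuation and overrides concurrency so new instances are prevented once the limit is reached.
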SoundCue /Game/Sounds/Fort_Doors/Constructed/Fort_Build_Door_Medium_01.Fort_Build_Door_Medium_01_Cue
root=Modulator_1
nodes [
 {id=Modulator_1, children [WavePlayer_1501]}
 {id=WavePlayer_1501, properties {soundWaveAssetPtr=&/Fort_Build_Door_Medium_01}}
]
properties {
 attenuationSettings=/Game/Sounds/Attenuation/Fort_Doors_Att
 bOverrideConcurrency=true
 maxDistance=2400
 soundClassObject=/Game/Sounds/FortSoundClassesAndModes/Fort_Player_Construction
 volumeMultiplier=1
 concurrencyOverrides {resolutionRule=PreventNew}
}
