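# Sound cue for the close sound of a medium constructed door in Fortnite building.
# A single stereo close wave plays through a Modulator node.
# It uses the shared door attenuation and the player construction sound class.
# Its concurrency override prevents new instances from starting once the limit is reached, rather than stopping ones already playing.
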SoundCue /Game/Sounds/Fort_Doors/Constructed/Fort_Build_Door_Medium_S_Close_Cue
root=Modulator_3
nodes [
 {id=Modulator_3, children [WavePlayer_16]}
 {id=WavePlayer_16, properties {soundWaveAssetPtr=&/door_constructed_close_stereo_05}}
]
properties {
 attenuationSettings=/Game/Sounds/Attenuation/Fort_Doors_Att
 bOverrideConcurrency=true
 maxDistance=2400
 soundClassObject=/Game/Sounds/FortSoundClassesAndModes/Fort_Player_Construction
 volumeMultiplier=1
 concurrencyOverrides {resolutionRule=PreventNew}
}
