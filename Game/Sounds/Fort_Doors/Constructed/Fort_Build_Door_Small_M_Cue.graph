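# Sound cue for a small constructed door in Fortnite building, playing a stereo door-open sample.
# A single wave player feeds a modulator at the root, with no random selection between waves.
# It uses the shared door attenuation and the player construction sound class.
# A concurrency override prevents new instances from starting once the limit is reached.
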SoundCue /Game/Sounds/Fort_Doors/Constructed/Fort_Build_Door_Small_M_Cue
root=Modulator_3
nodes [
 {id=Modulator_3, children [WavePlayer_15]}
 {id=WavePlayer_15, properties {soundWaveAssetPtr=&/door_constructed_open_stereo_03}}
]
properties {
 attenuationSettings=/Game/Sounds/Attenuation/Fort_Doors_Att
 bOverrideConcurrency=true
 maxDistance=2400
 soundClassObject=/Game/Sounds/FortSoundClassesAndModes/Fort_Player_Construction
 volumeMultiplier=1
 concurrencyOverrides {resolutionRule=PreventNew}
}
